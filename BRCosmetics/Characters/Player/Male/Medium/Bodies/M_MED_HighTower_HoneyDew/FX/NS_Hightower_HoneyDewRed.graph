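# Red Niagara effect for the HighTower HoneyDew player body cosmetic, a short staged burst of dots, smoke and glow around the character.
# Rising dots fill an ellipsoid while a first smoke layer appears, followed after a delay by brighter smoke and a sprite-plus-light burst.
# Smoke and glow follow skeletal mesh bones or sockets by spawn group, with glow particles offset to opposing sides by ID.
# Smoke uses flipbooks and curve-driven dynamic material parameters.
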NiagaraSystem /BRCosmetics/Characters/Player/Male/Medium/Bodies/M_MED_HighTower_HoneyDew/FX/NS_Hightower_HoneyDewRed
dependencies {
 usedDataInterfaces [&NiagaraDataInterfaceCurve, &NiagaraDataInterfaceColorCurve, &NiagaraDataInterfaceSkeletalMesh]
 usedDynamicInputs []
 usedModules [
  &self:SystemSpawnScript.SystemScriptSource.SystemScriptGraph.NiagaraNodeAssignment_0.SetVariables_#1
  &SystemState
 ]
}
emitters {
 Initial_Dots {
  emitterUpdateScript [
   &EmitterState(Life Cycle Mode=::NewEnumerator1, Loop Behavior=::NewEnumerator1, Loop Duration=0.45, UseLoopDelay=false)
   &SpawnRate(SpawnRate=50, Use Spawn Probability=false)
  ]
  particleSpawnScript [
   /Niagara/Modules/Spawn/Initialization/InitializeParticle(
    Lifetime=0.5
    Module.Write Color=true
    Module.Write Lifetime=true
    Module.Write Mass=true
    Module.Write Material Random=true
    Module.Write Position=false
    Module.Write Scale=false
    Module.Write SpriteRotation=false
    Module.Write SpriteSize=true
    Position=&SimulationPosition()
    Sprite Size=&Vector2DFromFloat(Value=&Multiply_Float(A=&UniformRangedFloat(Minimum=2, Maximum=4), B=1))
   )
   /Niagara/Modules/Spawn/Location/SphereLocation(
    MaskSpawnGroup=false
    Module.Override Randomness=false
    Non Uniform Scale=<0.6 0.6 1.25>
    Offset=<0 0 75>
    Sphere Radius=50
    Surface Only=false
   )
   &AddVelocity(
    Coordinate Space=::World
    Velocity=&MakeVector(X=0, Y=0, Z=&UniformRangedFloat(Minimum=100, Maximum=300))
   )
  ]
  particleUpdateScript [
   &ParticleState()
   &Drag(Drag=1.5, Use Linear Drag=true, Use Rotational Drag=false)
   &Color(
    Color=&ColorFromCurve(ColorCurve=DataInterface() {propertyValues {
       AlphaCurve {keys [<at 0 1 RCIM_Cubic>,<at 1 1 RCIM_Cubic>]}
       BlueCurve {keys [<at 0 0.1871 RCIM_Cubic>,<at 0.7701 0 RCIM_Cubic>]}
       GreenCurve {keys [<at 0 0 RCIM_Cubic>,<at 0.7701 0.2112 RCIM_Cubic>]}
       RedCurve {keys [<at 0 5 RCIM_Cubic>,<at 0.7701 0.5885 RCIM_Cubic>]}
      }})
    Scale Alpha=&FloatFromCurve(FloatCurve=DataInterface() {propertyValues {Curve {keys [{interpMode=RCIM_Cubic, leaveTangent=0.0979, tangentMode=RCTM_Break, time=0, value=1}, <at 1 0>]}}})
   )
   &ScaleSpriteSize(Scale Factor=&Vector2DFromFloat(Value=&FloatFromCurve(
      FloatCurve=DataInterface() {propertyValues {Curve {keys [
          {interpMode=RCIM_Cubic, leaveTangent=12.9788, tangentMode=RCTM_Break, time=0, value=0}
          {arriveTangent=-2.061, interpMode=RCIM_Cubic, leaveTangent=-4.7196, tangentMode=RCTM_Break, time=0.3046, value=1.1}
          {interpMode=RCIM_Cubic, tangentMode=RCTM_Break, time=1, value=0}
         ]}}}
      Scale Curve=&FloatFromCurve(FloatCurve=DataInterface() {propertyValues {Curve {keys [{interpMode=RCIM_Cubic, tangentMode=RCTM_Break, time=0, value=0.1}, <at 0.4 1>]}}})
     )))
   &SolveForcesAndVelocity(Clamp Velocity=false, Limit Acceleration=false)
  ]
  rendererClasses [&NiagaraSpriteRendererProperties]
  renderers [{rendererClass=&NiagaraSpriteRendererProperties}]
 }
 Initial_Smoke {
  emitterUpdateScript [
   &EmitterState(Life Cycle Mode=::NewEnumerator1, Loop Behavior=::NewEnumerator1, Loop Delay=0.45, Loop Duration=0.3, UseLoopDelay=false)
   &SpawnRate(SpawnRate=50, Use Spawn Probability=false)
  ]
  particleSpawnScript [
   &FN_SetSpawnGroup_v2(Number of Spawn Groups=15)
   &InitializeParticle(
    Color=&MultiplyLinearColor(
     Linear Color=&UniformRangedLinearColor(Minimum=<rgba 1 0.0318 0 0.02>, Maximum=<rgba 0.25 0.011 0 0.04>)
     Scale Factor=<1 1 1 1>
    )
    Lifetime=&UniformRangedFloat(Minimum=1.2, Maximum=1.8)
    Module.Write Color=true
    Module.Write Lifetime=true
    Module.Write Mass=false
    Module.Write Material Random=true
    Module.Write Position=false
    Module.Write Scale=false
    Module.Write SpriteRotation=true
    Module.Write SpriteSize=true
    Sprite Rotation=&UniformRangedFloat(Minimum=-360, Maximum=360)
    Sprite Size=&Vector2DFromFloat(Value=&Multiply_Float(A=&UniformRangedFloat(Minimum=22, Maximum=25), B=3))
   )
   &SystemLocation(MaskSpawnGroup=false, Offset=<0 0 100>) {enabled=false}
   /SetVariables(Particles.InitialRotationRate:NiagaraFloat=&UniformRangedFloat(Minimum=-60, Maximum=0)) {isSetParameters=true, module=SetVariables_#2}
   /SetVariables(Particles.RandomStartingFrame:NiagaraFloat=&UniformRangedFloat(Minimum=0, Maximum=0)) {isSetParameters=true, module=SetVariables_#3}
   &CameraOffset(Camera Offset Amount=&UniformRangedFloat(Minimum=4, Maximum=8))
  ]
  particleUpdateScript [
   &Drag(Drag=0.5, Use Linear Drag=true, Use Rotational Drag=false)
   &SkeletalMeshLocation(
    Bone / Socket Index=$Particles.SpawnGroup
    Bone Sampling Mode=::NewEnumerator4
    Inherit Velocity=false
    MaskSpawnGroup=false
    Skeletal Mesh=$System.Skeletal Mesh
   )
   &SolveForcesAndVelocity(Clamp Velocity=false, Limit Acceleration=false)
   &ParticleState()
   &SpriteRotationRate(Rotation Rate=$Particles.InitialRotationRate)
   &ScaleColor(
    Scale Alpha=&FloatFromCurve(FloatCurve=DataInterface() {propertyValues {Curve {keys [
         {interpMode=RCIM_Cubic, leaveTangent=12.0459, tangentMode=RCTM_Break, time=0, value=0}
         {arriveTangent=-2.4074, interpMode=RCIM_Cubic, leaveTangent=-3.8687, tangentMode=RCTM_Break, time=0.2, value=1}
         {interpMode=RCIM_Cubic, tangentMode=RCTM_User, time=0.5, value=0}
        ]}}})
    ScaleA=true
    ScaleRGB=false
   )
   /Niagara/Modules/Update/SubUV/SubUVAnimation(
    Number Of Frames=64
    SubUV Lookup Index=&Multiply_Float(A=&Add_Float(A=$Particles.NormalizedAge, B=$Particles.RandomStartingFrame), B=0.6)
   )
   &DynamicMaterialParameters(
    Index 0 Param 1=&FloatFromCurve(FloatCurve=DataInterface() {propertyValues {Curve {keys [<at 0.3 1 RCIM_Cubic>,<at 1 -0.1>]}}}, CurveIndex=$Particles.NormalizedLife)
    Index 0 Param 2=&FloatFromCurve(
     FloatCurve=DataInterface() {propertyValues {Curve {keys [<at 0 1 RCIM_Cubic>,<at 0.8 8.8 RCIM_Cubic>]}}}
     CurveIndex=$Particles.NormalizedLife
    )
    Index 0 Param 4=&FloatFromCurve(FloatCurve=DataInterface() {propertyValues {Curve {keys [<at 0.3 0>,<at 0.5 0.6>]}}}, CurveIndex=$Particles.NormalizedLife)
    Module.Param0WriteEnabled=true
    Module.Param1WriteEnabled=true
    Module.Param2WriteEnabled=true
    Module.Param3WriteEnabled=true
   )
  ]
  rendererClasses [&NiagaraSpriteRendererProperties]
  renderers [
   {rendererClass=&NiagaraSpriteRendererProperties}
   {rendererClass=&NiagaraSpriteRendererProperties, rendererIndex=1}
  ]
 }
 Large_Smoke {
  emitterUpdateScript [
   &EmitterState(Life Cycle Mode=::NewEnumerator1, Loop Behavior=::NewEnumerator1, Loop Delay=0.6, Loop Duration=0.3, UseLoopDelay=true)
   &SpawnRate(SpawnRate=50, Use Spawn Probability=false)
  ]
  particleSpawnScript [
   &FN_SetSpawnGroup_v2(Number of Spawn Groups=15)
   &InitializeParticle(
    Color=&MultiplyLinearColor(
     Linear Color=&UniformRangedLinearColor(Minimum=<rgba 1 0 0.063 0.5>, Maximum=<rgba 0.25 0 0.0107 0.1>)
     Scale Factor=<1 1 1 1>
    )
    Lifetime=&UniformRangedFloat(Minimum=0.5, Maximum=1.2)
    Module.Write Color=true
    Module.Write Lifetime=true
    Module.Write Mass=false
    Module.Write Material Random=true
    Module.Write Position=false
    Module.Write Scale=false
    Module.Write SpriteRotation=true
    Module.Write SpriteSize=true
    Sprite Rotation=&UniformRangedFloat(Minimum=-360, Maximum=360)
    Sprite Size=&Vector2DFromFloat(Value=&Multiply_Float(A=&UniformRangedFloat(Minimum=22, Maximum=25), B=3))
   )
   &SystemLocation(MaskSpawnGroup=false, Offset=<0 0 100>) {enabled=false}
   /SetVariables(Particles.InitialRotationRate:NiagaraFloat=&UniformRangedFloat(Minimum=-60, Maximum=0)) {isSetParameters=true, module=SetVariables_#2}
   /SetVariables(Particles.RandomStartingFrame:NiagaraFloat=&UniformRangedFloat(Minimum=0, Maximum=0)) {isSetParameters=true, module=SetVariables_#3}
   &CameraOffset(Camera Offset Amount=&UniformRangedFloat(Minimum=4, Maximum=8))
  ]
  particleUpdateScript [
   &Drag(Drag=0.5, Use Linear Drag=true, Use Rotational Drag=false)
   &SkeletalMeshLocation(
    Bone / Socket Index=$Particles.SpawnGroup
    Bone Sampling Mode=::NewEnumerator4
    Inherit Velocity=false
    MaskSpawnGroup=false
    Skeletal Mesh=$System.Skeletal Mesh
   )
   &SolveForcesAndVelocity(Clamp Velocity=false, Limit Acceleration=false)
   &ParticleState()
   &SpriteRotationRate(Rotation Rate=$Particles.InitialRotationRate)
   &ScaleColor(
    Scale Alpha=&FloatFromCurve(FloatCurve=DataInterface() {propertyValues {Curve {keys [
         {interpMode=RCIM_Cubic, leaveTangent=12.0459, tangentMode=RCTM_Break, time=0, value=0}
         {arriveTangent=-2.4074, interpMode=RCIM_Cubic, leaveTangent=-3.8687, tangentMode=RCTM_Break, time=0.2, value=1}
         {interpMode=RCIM_Cubic, tangentMode=RCTM_User, time=0.5, value=0}
        ]}}})
    ScaleA=true
    ScaleRGB=false
   )
   &SubUVAnimation(
    Number Of Frames=64
    SubUV Lookup Index=&Multiply_Float(A=&Add_Float(A=$Particles.NormalizedAge, B=$Particles.RandomStartingFrame), B=0.6)
   )
   &DynamicMaterialParameters(
    Index 0 Param 1=&FloatFromCurve(FloatCurve=DataInterface() {propertyValues {Curve {keys [<at 0.3 1 RCIM_Cubic>,<at 1 -0.1>]}}}, CurveIndex=$Particles.NormalizedLife)
    Index 0 Param 2=&FloatFromCurve(
     FloatCurve=DataInterface() {propertyValues {Curve {keys [<at 0 1 RCIM_Cubic>,<at 0.8 8.8 RCIM_Cubic>]}}}
     CurveIndex=$Particles.NormalizedLife
    )
    Index 0 Param 4=&FloatFromCurve(FloatCurve=DataInterface() {propertyValues {Curve {keys [<at 0.3 0>,<at 0.5 0.6>]}}}, CurveIndex=$Particles.NormalizedLife)
    Module.Param0WriteEnabled=true
    Module.Param1WriteEnabled=true
    Module.Param2WriteEnabled=true
    Module.Param3WriteEnabled=true
   )
  ]
  rendererClasses [&NiagaraSpriteRendererProperties]
  renderers [
   {rendererClass=&NiagaraSpriteRendererProperties}
   {rendererClass=&NiagaraSpriteRendererProperties, rendererIndex=1}
  ]
 }
 Red_Decal {
  emitterUpdateScript [
   &EmitterState(Life Cycle Mode=::NewEnumerator1, Loop Behavior=::NewEnumerator1, Loop Delay=0.6, Loop Duration=1, UseLoopDelay=true)
   &SpawnBurst_Instantaneous(Spawn Count=30, Use Spawn Probability=false)
  ]
  particleSpawnScript [
   &FN_SetSpawnGroup_v2(Number of Spawn Groups=15)
   &InitializeParticle(
    Lifetime=0.95
    Module.Write Color=true
    Module.Write Lifetime=true
    Module.Write Mass=true
    Module.Write Material Random=true
    Module.Write Position=false
    Module.Write Scale=false
    Module.Write SpriteRotation=false
    Module.Write SpriteSize=true
    Position=&SimulationPosition()
    Sprite Size=&Vector2DFromFloat(Value=75)
   )
  ]
  particleUpdateScript [
   &ParticleState()
   &SkeletalMeshLocation(
    Apply Sampled Orientation to Mesh Particles=true
    Bone / Socket Index=$Particles.SpawnGroup
    Bone Sampling Mode=::NewEnumerator4
    Inherit Velocity=false
    MaskSpawnGroup=false
    Sampled Position Offset=&MakeCustomVectorFromBool(Bool=HlslExpression() {hlslExpression="Particles.UniqueID < 15 ? 1:0"}, True Vector=<0 20 1>, False Vector=<1 -20 0>)
    Skeletal Mesh=$System.Skeletal Mesh
   )
   &SolveForcesAndVelocity(Clamp Velocity=false, Limit Acceleration=false)
   &Color(
    Color=<rgba 2 0 0.018 1>
    Scale Color=&VectorFromFloat(Value=&FloatFromCurve(FloatCurve=DataInterface() {propertyValues {Curve {keys [
          {interpMode=RCIM_Cubic, leaveTangent=12.0459, tangentMode=RCTM_Break, time=0, value=0}
          {arriveTangent=-2.4074, interpMode=RCIM_Cubic, leaveTangent=-3.8687, tangentMode=RCTM_Break, time=0.2, value=1}
          <at 1 0>
         ]}}}))
   )
   /SetVariables(Particles.LightColor=&MultiplyLinearColor(
     Linear Color=<rgba 3 0 0.0562 1>
     Scale Factor=&MakeVector4_FromFloat(Float=&FloatFromCurve(FloatCurve=DataInterface() {propertyValues {Curve {keys [
           {interpMode=RCIM_Cubic, leaveTangent=12.0459, tangentMode=RCTM_Break, time=0, value=0}
           {arriveTangent=-2.4074, interpMode=RCIM_Cubic, leaveTangent=-3.8687, tangentMode=RCTM_Break, time=0.2, value=1}
           <at 1 0>
          ]}}}))
    )) {isSetParameters=true}
  ]
  rendererClasses [&NiagaraSpriteRendererProperties, &NiagaraLightRendererProperties]
  renderers [
   {rendererClass=&NiagaraSpriteRendererProperties}
   {rendererClass=&NiagaraLightRendererProperties, rendererIndex=1}
  ]
 }
}
summary {
 systemName=NS_Hightower_HoneyDewRed
 emitters [
  {emitterName=Red_Decal, rendererClasses [&NiagaraSpriteRendererProperties, &NiagaraLightRendererProperties]}
  {emitterName=Large_Smoke, rendererClasses [&NiagaraSpriteRendererProperties]}
  {emitterName=Initial_Dots, rendererClasses [&NiagaraSpriteRendererProperties]}
  {emitterName=Initial_Smoke, rendererClasses [&NiagaraSpriteRendererProperties]}
 ]
 userVariables [
  User.bIsFrontEndPreview {type=NiagaraBool}
  User.Color {type=LinearColor, description="", defaultValue {struct=LinearColor, value=<rgba 1 1 1 1>}}
  User.PlayerSpeed {type=NiagaraFloat, defaultValue=0, description=""}
  User.RibbonIDOffset {type=NiagaraInt32}
 ]
}
