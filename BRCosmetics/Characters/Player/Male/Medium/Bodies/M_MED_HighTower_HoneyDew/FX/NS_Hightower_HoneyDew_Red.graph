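# Niagara effect for the HighTower HoneyDew player body cosmetic: a short red-orange burst of dots, smoke and glow on the character.
# Rising dots and tapering smoke lead into delayed, spinning smoke puffs with animated sub-images.
# Smoke and Red_Decal glow sprites sample skeletal mesh bones by spawn group, with large puffs and glows mirrored either side of 16 indices.
# An emitter Buildup curve flashes the glow and computed light color, though only sprite renderers exist.
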NiagaraSystem /BRCosmetics/Characters/Player/Male/Medium/Bodies/M_MED_HighTower_HoneyDew/FX/NS_Hightower_HoneyDew_Red
dependencies {
 usedDataInterfaces [
  /Script/Niagara.NiagaraDataInterfaceArrayBool
  &NiagaraDataInterfaceCurve
  &NiagaraDataInterfaceColorCurve
  &NiagaraDataInterfaceVector2DCurve
  &NiagaraDataInterfaceSkeletalMesh
 ]
 usedDynamicInputs []
 usedModules [
  &self:SystemSpawnScript.SystemScriptSource.SystemScriptGraph.NiagaraNodeAssignment_0.SetVariables_#1
  &SystemState
 ]
}
emitters {
 Initial_Dots {
  emitterUpdateScript [
   &EmitterState(Life Cycle Mode=::NewEnumerator1, Loop Behavior=::NewEnumerator1, Loop Duration=0.5, UseLoopDelay=false)
   &SpawnRate(SpawnRate=50, Use Spawn Probability=false)
  ]
  particleSpawnScript [
   /Niagara/Modules/Spawn/Initialization/InitializeParticle(
    Lifetime=&UniformRangedFloat(Minimum=0.4, Maximum=0.6)
    Module.Write Color=true
    Module.Write Lifetime=true
    Module.Write Mass=true
    Module.Write Material Random=true
    Module.Write Position=false
    Module.Write Scale=false
    Module.Write SpriteRotation=false
    Module.Write SpriteSize=true
    Position=&SimulationPosition()
    Sprite Size=&Vector2DFromFloat(Value=&Multiply_Float(A=&UniformRangedFloat(Minimum=2, Maximum=4), B=1))
   )
   /Niagara/Modules/Spawn/Location/SphereLocation(
    MaskSpawnGroup=false
    Module.Override Randomness=false
    Non Uniform Scale=<0.6 0.6 1.25>
    Offset=<0 0 75>
    Sphere Radius=50
    Surface Only=false
   )
   &AddVelocity(
    Coordinate Space=::World
    Velocity=&MakeVector(X=0, Y=0, Z=&UniformRangedFloat(Minimum=30, Maximum=100))
   )
  ]
  particleUpdateScript [
   &ParticleState()
   &Color(
    Color=&ColorFromCurve(ColorCurve=DataInterface() {propertyValues {
       AlphaCurve {keys [<at 0 1 RCIM_Cubic>,<at 1 1 RCIM_Cubic>]}
       BlueCurve {keys [<at 0 0 RCIM_Cubic>,<at 0.7701 0 RCIM_Cubic>]}
       GreenCurve {keys [<at 0 0.0256 RCIM_Cubic>,<at 0.7701 0.2835 RCIM_Cubic>]}
       RedCurve {keys [<at 0 1 RCIM_Cubic>,<at 0.7701 0.5885 RCIM_Cubic>]}
      }})
    Scale Alpha=&FloatFromCurve(FloatCurve=DataInterface() {propertyValues {Curve {keys [{interpMode=RCIM_Cubic, leaveTangent=0.0979, tangentMode=RCTM_Break, time=0, value=1}, <at 1 0>]}}})
   )
   &ScaleSpriteSize(Scale Factor=&Vector2DFromFloat(Value=&FloatFromCurve(
      FloatCurve=DataInterface() {propertyValues {Curve {keys [
          {interpMode=RCIM_Cubic, leaveTangent=12.9788, tangentMode=RCTM_Break, time=0, value=0}
          {arriveTangent=-2.061, interpMode=RCIM_Cubic, leaveTangent=-4.7196, tangentMode=RCTM_Break, time=0.3046, value=1.1}
          {interpMode=RCIM_Cubic, tangentMode=RCTM_Break, time=1, value=0}
         ]}}}
      Scale Curve=&FloatFromCurve(FloatCurve=DataInterface() {propertyValues {Curve {keys [{interpMode=RCIM_Cubic, tangentMode=RCTM_Break, time=0, value=0.1}, <at 0.4 1>]}}})
     )))
   &ScaleVelocity(Velocity Scale=&MakeVector(
     X=1
     Y=1
     Z=&FloatFromCurve(FloatCurve=DataInterface() {propertyValues {Curve {keys [<at 0 1 RCIM_Cubic>, {arriveTangent=3.6879, interpMode=RCIM_Cubic, leaveTangent=3.6879, tangentMode=RCTM_User, time=1, value=2.5}]}}})
    ))
   &SolveForcesAndVelocity(Clamp Velocity=false, Limit Acceleration=false)
  ]
  rendererClasses [&NiagaraSpriteRendererProperties]
  renderers [{rendererClass=&NiagaraSpriteRendererProperties}]
 }
 Initial_Smoke {
  emitterUpdateScript [
   &EmitterState(Life Cycle Mode=::NewEnumerator1, Loop Behavior=::NewEnumerator1, Loop Duration=0.75, UseLoopDelay=false)
   &SpawnRate(
    SpawnRate=&FloatFromCurve(
     FloatCurve=DataInterface() {propertyValues {Curve {keys [<at 0 1 RCIM_Cubic>, {arriveTangent=-1.85, interpMode=RCIM_Cubic, leaveTangent=-1.85, tangentMode=RCTM_User, time=1, value=0}]}}}
     CurveIndex=$Emitter.NormalizedLoopAge
     Scale Curve=50
    )
    Use Spawn Probability=false
   )
  ]
  particleSpawnScript [
   &FN_SetSpawnGroup_v2(Number of Spawn Groups=17)
   &InitializeParticle(
    Color=&MultiplyLinearColor(
     Linear Color=&UniformRangedLinearColor(Minimum=<rgba 1 0.0552 0 1>, Maximum=<rgba 2 0.0335 0 1>, Color Channel Mode=::NewEnumerator1)
     Scale Factor=<1 1 1 1>
    )
    Lifetime=&FloatFromCurve(FloatCurve=DataInterface() {propertyValues {Curve {keys [<at 0 1.25 RCIM_Cubic>,<at 1 0.2>]}}}, CurveIndex=$Emitter.NormalizedLoopAge)
    Module.Write Color=true
    Module.Write Lifetime=true
    Module.Write Mass=false
    Module.Write Material Random=true
    Module.Write Position=false
    Module.Write Scale=false
    Module.Write SpriteRotation=true
    Module.Write SpriteSize=true
    Sprite Size=&UniformRangedVector2D(Minimum=<15 40>, Maximum=<35 80>)
   )
   &CameraOffset(Camera Offset Amount=&UniformAOrBFloat(A=5, B=16))
   &DynamicMaterialParameters(
    Index 0 Param 2=$Particles.MaterialRandom
    Module.Param0WriteEnabled=false
    Module.Param1WriteEnabled=true
    Module.Param2WriteEnabled=false
    Module.Param3WriteEnabled=false
   ) {module=DynamicMaterialParameters001}
  ]
  particleUpdateScript [
   &ParticleState()
   &ScaleColor(
    Scale Alpha=&FloatFromCurve(FloatCurve=DataInterface() {propertyValues {Curve {keys [
         {arriveTangent=-8.0953, interpMode=RCIM_Cubic, leaveTangent=-8.0952, tangentMode=RCTM_User, time=0, value=1}
         {arriveTangent=-0.8371, interpMode=RCIM_Cubic, leaveTangent=-0.8371, time=0.1955, value=0.0807}
         {arriveTangent=1.1427, interpMode=RCIM_Cubic, leaveTangent=1.1427, time=0.8123, value=0.32}
         {interpMode=RCIM_Cubic, leaveTangent=12.0487, time=1, value=1}
        ]}}})
    ScaleA=true
    ScaleRGB=false
   )
   &SkeletalMeshLocation(
    Bone / Socket Index=$Particles.SpawnGroup
    Bone Sampling Mode=::NewEnumerator4
    Inherit Velocity=false
    MaskSpawnGroup=false
    Sampled Position Offset=<0 0 -6>
    Skeletal Mesh=$System.Skeletal Mesh
   )
   &DynamicMaterialParameters(
    Index 0 Param 1=&FloatFromCurve(
     FloatCurve=DataInterface() {propertyValues {Curve {keys [
         {arriveTangent=4.5199, interpMode=RCIM_Cubic, leaveTangent=4.5199, tangentMode=RCTM_User, time=0, value=0}
         {arriveTangent=0.8527, interpMode=RCIM_Cubic, leaveTangent=0.8527, tangentMode=RCTM_User, time=0.2951, value=0.6871}
         {arriveTangent=0.3142, interpMode=RCIM_Cubic, leaveTangent=0.3142, tangentMode=RCTM_User, time=1, value=1}
        ]}}}
     Scale Curve=2
    )
    Module.Param0WriteEnabled=true
    Module.Param1WriteEnabled=false
    Module.Param2WriteEnabled=false
    Module.Param3WriteEnabled=false
   )
   &ScaleSpriteSize(Scale Factor=&Vector2DFromCurve(
     Vector2Curve=DataInterface() {propertyValues {XCurve {keys [<at 0 1 RCIM_Cubic>,<at 1 0.6 RCIM_Cubic>]}, YCurve {keys [<at 0 0.3 RCIM_Cubic>,<at 1 1 RCIM_Cubic>]}}}
     ScaleCurve=<1 1>
    ))
  ]
  rendererClasses [&NiagaraSpriteRendererProperties]
  renderers [{rendererClass=&NiagaraSpriteRendererProperties}]
 }
 Large_Smoke {
  emitterUpdateScript [
   &EmitterState(Life Cycle Mode=::NewEnumerator1, Loop Behavior=::NewEnumerator1, Loop Delay=0.5, Loop Duration=0.25, UseLoopDelay=true)
   &SpawnRate(SpawnRate=150, Use Spawn Probability=false)
   &SpawnBurst_Instantaneous(Spawn Count=10, Use Spawn Probability=false)
  ]
  particleSpawnScript [
   &FN_SetSpawnGroup_v2(Number of Spawn Groups=32)
   &InitializeParticle(
    Color=&MultiplyLinearColor(
     Linear Color=&UniformRangedLinearColor(Minimum=<rgba 0.5 0 0.0045 0.75>, Maximum=<rgba 0.75 0.0176 0 1>)
     Scale Factor=<1 1 1 1>
    )
    Lifetime=&UniformRangedFloat(Minimum=0.8, Maximum=1.5)
    Module.Write Color=true
    Module.Write Lifetime=true
    Module.Write Mass=false
    Module.Write Material Random=true
    Module.Write Position=false
    Module.Write Scale=false
    Module.Write SpriteRotation=true
    Module.Write SpriteSize=true
    Sprite Rotation=&UniformRangedFloat(Minimum=-360, Maximum=360)
    Sprite Size=&Multiply_Vector2(
     A=&UniformRangedVector2D(Minimum=<30 30>, Maximum=<60 60>)
     B=&MakeVector2D(
      X=&MakeCustomFloatFromBool(Bool=HlslExpression() {hlslExpression="(Particles.SpawnGroup && 1) == 0"}, True Float=1, False Float=-1)
      Y=1
     )
    )
   )
   /SetVariables(Particles.InitialRotationRate:NiagaraFloat=&MakeCustomFloatFromBool(Bool=HlslExpression() {hlslExpression="(Particles.SpawnGroup && 1) == 0"}, True Float=-200, False Float=200)) {isSetParameters=true}
   /SetVariables(Particles.InitialOffset:Vector3f=&UniformRangedVector(Minimum=<-5 -5 -5>, Maximum=<5 5 5>)) {isSetParameters=true}
   &SkeletalMeshLocation(
    Bone / Socket Index=&ModuloInt(A=$Particles.SpawnGroup, B=16)
    Bone Sampling Mode=::NewEnumerator4
    Inherit Velocity=false
    MaskSpawnGroup=false
    Sampled Position Offset=&Add_Vector(
     A=$Particles.InitialOffset
     B=&MakeCustomVectorFromBool(
      Bool=&SetBoolByIntComparison(A=$Particles.SpawnGroup, B=16, Comparison Type=::NewEnumerator0)
      True Vector=<8 0 0>
      False Vector=<-8 0 0>
     )
    )
    Sampled Position Offset Coordinate Space:ENiagaraMeshTransforms=::NewEnumerator3
    Skeletal Mesh=$System.Skeletal Mesh
   ) {module=SkeletalMeshLocation001}
   &CameraOffset(Camera Offset Amount=&UniformRangedFloat(Minimum=7, Maximum=12))
   &AddVelocity(Velocity=&MakeVector(X=0, Y=0, Z=&UniformRangedFloat(Minimum=15, Maximum=25)))
  ]
  particleUpdateScript [
   &Drag(Use Linear Drag=true, Use Rotational Drag=false)
   &SolveForcesAndVelocity(Clamp Velocity=false, Limit Acceleration=false)
   &ParticleState()
   &SpriteRotationRate(Rotation Rate=$Particles.InitialRotationRate)
   &ScaleColor(
    Scale Alpha=&FloatFromCurve(FloatCurve=DataInterface() {propertyValues {Curve {keys [
         {arriveTangent=7.9734, interpMode=RCIM_Cubic, leaveTangent=7.9734, tangentMode=RCTM_User, time=0, value=0}
         {arriveTangent=2.3284, interpMode=RCIM_Cubic, leaveTangent=2.3284, time=0.2929, value=0.8541}
         <at 0.859 2>
         {arriveTangent=-0.9204, interpMode=RCIM_Cubic, leaveTangent=-0.9204, tangentMode=RCTM_User, time=1, value=0.8}
        ]}}})
    ScaleA=true
    ScaleRGB=false
   )
   /SetVariables(Particles.SubImageIndex=&Lerp_Float(A=&Lerp_Float(A=5, B=12, Alpha=$Particles.Initial.MaterialRandom), B=63, Alpha=$Particles.NormalizedAge)) {isSetParameters=true}
   &DynamicMaterialParameters(
    Index 0 Param 1=&FloatFromCurve(FloatCurve=DataInterface() {propertyValues {Curve {keys [<at 0.3 1 RCIM_Cubic>,<at 1 -0.1>]}}}, CurveIndex=$Particles.NormalizedLife)
    Index 0 Param 2=&FloatFromCurve(
     FloatCurve=DataInterface() {propertyValues {Curve {keys [<at 0 1 RCIM_Cubic>,<at 0.8 8.8 RCIM_Cubic>]}}}
     CurveIndex=$Particles.NormalizedLife
    )
    Index 0 Param 4=&FloatFromCurve(FloatCurve=DataInterface() {propertyValues {Curve {keys [<at 0.3 0>,<at 0.5 0.6>]}}}, CurveIndex=$Particles.NormalizedLife)
    Module.Param0WriteEnabled=true
    Module.Param1WriteEnabled=true
    Module.Param2WriteEnabled=true
    Module.Param3WriteEnabled=true
   )
   &ScaleSpriteSize(Scale Factor=&Vector2DFromFloat(Value=&FloatFromCurve(FloatCurve=DataInterface() {propertyValues {Curve {keys [
          {arriveTangent=1.0472, interpMode=RCIM_Cubic, leaveTangent=1.0472, tangentMode=RCTM_User, time=0, value=0.2}
          {arriveTangent=0.1666, interpMode=RCIM_Cubic, leaveTangent=0.1666, tangentMode=RCTM_User, time=1, value=1}
         ]}}})))
  ]
  rendererClasses [&NiagaraSpriteRendererProperties]
  renderers [{rendererClass=&NiagaraSpriteRendererProperties}]
 }
 Red_Decal {
  emitterSpawnScript [/SetVariables(Emitter.LightOn=DataInterface() {propertyValues {
      GpuSyncMode=SyncCpuToGpu
      MaxElements=0
      BoolData [true,false,true,false,false,false,false,false,true,true,true,true,false,false,false,false,true,true]
     }}) {isSetParameters=true}]
  emitterUpdateScript [
   &EmitterState(Life Cycle Mode=::NewEnumerator1, Loop Behavior=::NewEnumerator1, Loop Delay=0.6, Loop Duration=1, UseLoopDelay=false)
   /SetVariables(Emitter.Buildup:NiagaraFloat=&FloatFromCurve(
     FloatCurve=DataInterface() {propertyValues {Curve {keys [
         <at 0 0 RCIM_Cubic>
         {arriveTangent=0.9901, interpMode=RCIM_Cubic, leaveTangent=0.9901, tangentMode=RCTM_User, time=0.6, value=0.25}
         {arriveTangent=-2.4074, interpMode=RCIM_Cubic, leaveTangent=-1.7251, tangentMode=RCTM_Break, time=0.64, value=1}
         <at 1.75 0>
        ]}}}
     CurveIndex=$Emitter.Age
    )) {isSetParameters=true}
   &SpawnBurst_Instantaneous(Spawn Count=32, Use Spawn Probability=false)
  ]
  particleSpawnScript [
   &FN_SetSpawnGroup_v2(Number of Spawn Groups=32)
   /SetVariables(Particles.LightEnabled:NiagaraBool=&self:GetBoolArray(LightOn:NiagaraDataInterfaceArrayBool=$Emitter.LightOn, SpawnGroupInt:NiagaraInt32=$Particles.SpawnGroup)) {isSetParameters=true}
   &InitializeParticle(
    Lifetime=1.75
    Module.Write Color=true
    Module.Write Lifetime=true
    Module.Write Mass=true
    Module.Write Material Random=true
    Module.Write Position=false
    Module.Write Scale=false
    Module.Write SpriteRotation=false
    Module.Write SpriteSize=true
    Position=&SimulationPosition()
    Sprite Size=&Vector2DFromFloat(Value=35)
   )
   &CameraOffset(Camera Offset Amount=20)
   /SetVariables(Particles.LightExp=5) {isSetParameters=true}
  ]
  particleUpdateScript [
   &ParticleState()
   &SkeletalMeshLocation(
    Bone / Socket Index=&ModuloInt(A=$Particles.SpawnGroup, B=16)
    Bone Sampling Mode=::NewEnumerator4
    Inherit Velocity=false
    MaskSpawnGroup=false
    Sampled Position Offset=&MakeCustomVectorFromBool(
     Bool=&SetBoolByIntComparison(A=$Particles.SpawnGroup, B=16, Comparison Type=::NewEnumerator0)
     True Vector=<8 0 0>
     False Vector=<-8 0 0>
    )
    Sampled Position Offset Coordinate Space:ENiagaraMeshTransforms=::NewEnumerator3
    Skeletal Mesh=$System.Skeletal Mesh
   )
   /SetVariables(Particles.LightPosition:Vector3f=$Output.SkeletalMeshLocation.SampledPosition) {isSetParameters=true}
   &Color(
    Color=<rgba 3 0.139 0 1>
    Scale Alpha=&FloatFromCurve(FloatCurve=DataInterface() {propertyValues {Curve {keys [
         {interpMode=RCIM_Cubic, tangentMode=RCTM_User, time=0, value=0}
         {interpMode=RCIM_Cubic, tangentMode=RCTM_User, time=0.3, value=1}
         {interpMode=RCIM_Cubic, tangentMode=RCTM_User, time=0.6, value=1}
         {interpMode=RCIM_Cubic, tangentMode=RCTM_User, time=1, value=0}
        ]}}})
    Scale Color=&VectorFromFloat(Value=$Emitter.Buildup)
   )
   /SetVariables(Particles.LightColor=&MultiplyLinearColor(Linear Color=<rgba 3 0.139 0 1>, Scale Factor=&MakeVector4_FromFloat(Float=$Emitter.Buildup))) {isSetParameters=true}
   &DynamicMaterialParameters(
    Index 0 Param 1=&Lerp_Float(A=1.5, B=0.5, Alpha=$Emitter.Buildup)
    Module.Param0WriteEnabled=true
    Module.Param1WriteEnabled=false
    Module.Param2WriteEnabled=false
    Module.Param3WriteEnabled=false
   )
  ]
  rendererClasses [&NiagaraSpriteRendererProperties]
  renderers [{rendererClass=&NiagaraSpriteRendererProperties}]
 }
}
summary {
 systemName=NS_Hightower_HoneyDew_Red
 emitters [
  {emitterName=Red_Decal, rendererClasses [&NiagaraSpriteRendererProperties]}
  {emitterName=Large_Smoke, rendererClasses [&NiagaraSpriteRendererProperties]}
  {emitterName=Initial_Dots, rendererClasses [&NiagaraSpriteRendererProperties]}
  {emitterName=Initial_Smoke, rendererClasses [&NiagaraSpriteRendererProperties]}
 ]
 userVariables [
  User.bIsFrontEndPreview {type=NiagaraBool}
  User.Color {type=LinearColor, description="", defaultValue {struct=LinearColor, value=<rgba 1 1 1 1>}}
  User.PlayerSpeed {type=NiagaraFloat, defaultValue=0, description=""}
  User.RibbonIDOffset {type=NiagaraInt32}
 ]
}
