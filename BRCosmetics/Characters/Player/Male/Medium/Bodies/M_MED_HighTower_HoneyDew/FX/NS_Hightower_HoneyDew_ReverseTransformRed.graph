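# Niagara system for the HighTower HoneyDew character's red reverse-transform effect, flashing glowing decal sprites and lights along the skeletal mesh.
# A single burst spawns 32 spawn-grouped particles that follow 16 bone/socket indices, paired with mirrored offsets.
# A per-group bool array sets which particles enable their light.
# An age-driven Buildup curve slowly ramps, spikes, then fades, scaling sprite color, light color and a dynamic material parameter.
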NiagaraSystem /BRCosmetics/Characters/Player/Male/Medium/Bodies/M_MED_HighTower_HoneyDew/FX/NS_Hightower_HoneyDew_ReverseTransformRed
dependencies {
 usedDataInterfaces [/Script/Niagara.NiagaraDataInterfaceArrayBool, &NiagaraDataInterfaceCurve, &NiagaraDataInterfaceSkeletalMesh]
 usedDynamicInputs []
 usedModules [
  &self:SystemSpawnScript.SystemScriptSource.SystemScriptGraph.NiagaraNodeAssignment_0.SetVariables_#1
  &SystemState
 ]
}
emitters {Red_Decal {
  emitterSpawnScript [/SetVariables(Emitter.LightOn=DataInterface() {propertyValues {
      GpuSyncMode=SyncCpuToGpu
      MaxElements=0
      BoolData [true,false,true,false,false,false,false,false,true,true,true,true,false,false,false,false,true,true]
     }}) {isSetParameters=true}]
  emitterUpdateScript [
   &EmitterState(Life Cycle Mode=::NewEnumerator1, Loop Behavior=::NewEnumerator1, Loop Delay=0.6, Loop Duration=1, UseLoopDelay=false)
   /SetVariables(Emitter.Buildup:NiagaraFloat=&FloatFromCurve(
     FloatCurve=DataInterface() {propertyValues {Curve {keys [
         <at 0 0 RCIM_Cubic>
         {arriveTangent=0.9901, interpMode=RCIM_Cubic, leaveTangent=0.9901, tangentMode=RCTM_User, time=0.6, value=0.25}
         {arriveTangent=-2.4074, interpMode=RCIM_Cubic, leaveTangent=-1.7251, tangentMode=RCTM_Break, time=0.64, value=1}
         <at 1.75 0>
        ]}}}
     CurveIndex=$Emitter.Age
    )) {isSetParameters=true}
   &SpawnBurst_Instantaneous(Spawn Count=32, Use Spawn Probability=false)
  ]
  particleSpawnScript [
   &FN_SetSpawnGroup_v2(Number of Spawn Groups=32)
   /SetVariables(Particles.LightEnabled:NiagaraBool=&self:GetBoolArray(LightOn:NiagaraDataInterfaceArrayBool=$Emitter.LightOn, SpawnGroupInt:NiagaraInt32=$Particles.SpawnGroup)) {isSetParameters=true}
   /Niagara/Modules/Spawn/Initialization/InitializeParticle(
    Lifetime=1.75
    Module.Write Color=true
    Module.Write Lifetime=true
    Module.Write Mass=true
    Module.Write Material Random=true
    Module.Write Position=false
    Module.Write Scale=false
    Module.Write SpriteRotation=false
    Module.Write SpriteSize=true
    Position=&SimulationPosition()
    Sprite Size=&Vector2DFromFloat(Value=35)
   )
   &CameraOffset(Camera Offset Amount=20)
   /SetVariables(Particles.LightExp=5) {isSetParameters=true}
  ]
  particleUpdateScript [
   &ParticleState()
   &SkeletalMeshLocation(
    Bone / Socket Index=&ModuloInt(A=$Particles.SpawnGroup, B=16)
    Bone Sampling Mode=::NewEnumerator4
    Inherit Velocity=false
    MaskSpawnGroup=false
    Sampled Position Offset=&MakeCustomVectorFromBool(
     Bool=&SetBoolByIntComparison(A=$Particles.SpawnGroup, B=16, Comparison Type=::NewEnumerator0)
     True Vector=<8 0 0>
     False Vector=<-8 0 0>
    )
    Sampled Position Offset Coordinate Space:ENiagaraMeshTransforms=::NewEnumerator3
    Skeletal Mesh=$System.Skeletal Mesh
   )
   /SetVariables(Particles.LightPosition:Vector3f=$Output.SkeletalMeshLocation.SampledPosition) {isSetParameters=true}
   &Color(
    Color=<rgba 3 0.139 0 1>
    Scale Alpha=&FloatFromCurve(FloatCurve=DataInterface() {propertyValues {Curve {keys [
         {interpMode=RCIM_Cubic, tangentMode=RCTM_User, time=0, value=0}
         {interpMode=RCIM_Cubic, tangentMode=RCTM_User, time=0.3, value=1}
         {interpMode=RCIM_Cubic, tangentMode=RCTM_User, time=0.6, value=1}
         {interpMode=RCIM_Cubic, tangentMode=RCTM_User, time=1, value=0}
        ]}}})
    Scale Color=&VectorFromFloat(Value=$Emitter.Buildup)
   )
   /SetVariables(Particles.LightColor=&MultiplyLinearColor(Linear Color=<rgba 3 0.139 0 1>, Scale Factor=&MakeVector4_FromFloat(Float=$Emitter.Buildup))) {isSetParameters=true}
   &DynamicMaterialParameters(
    Index 0 Param 1=&Lerp_Float(A=1.5, B=0.5, Alpha=$Emitter.Buildup)
    Module.Param0WriteEnabled=true
    Module.Param1WriteEnabled=false
    Module.Param2WriteEnabled=false
    Module.Param3WriteEnabled=false
   )
  ]
  rendererClasses [&NiagaraSpriteRendererProperties, &NiagaraLightRendererProperties]
  renderers [
   {rendererClass=&NiagaraSpriteRendererProperties}
   {rendererClass=&NiagaraLightRendererProperties, rendererIndex=1}
  ]
 }}
summary {
 systemName=NS_Hightower_HoneyDew_ReverseTransformRed
 emitters [{emitterName=Red_Decal, rendererClasses [&NiagaraSpriteRendererProperties, &NiagaraLightRendererProperties]}]
 userVariables [
  User.bIsFrontEndPreview {type=NiagaraBool}
  User.Color {type=LinearColor, description="", defaultValue {struct=LinearColor, value=<rgba 1 1 1 1>}}
  User.PlayerSpeed {type=NiagaraFloat, defaultValue=0, description=""}
  User.RibbonIDOffset {type=NiagaraInt32}
 ]
}
